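# Control Rig for the PeelyMech medium male head part, procedurally driving its dynamic blender-head bones.
# A transform constraint holds the blender base between spine_05 and neck_01, weighted with maintained offset.
# The head's rotation relative to spine_05 is reduced to its X Euler component and added locally to dyn_blenderHead.
# The `blender` rotator variable adds further local rotation on the child extraSpin bone, layering spin over the head-follow motion.
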ControlRigBlueprint /BRCosmetics/Characters/Player/Male/Medium/Bodies/M_MED_PeelyMech/Parts/Male_Medium_PeelyMech_Head_CtrlRig
graphs [
 RigVMFunctionLibrary {links [], nodes []}
 RigVMModel {nodes [
   n1 = &RigUnit_BeginExecution()
   /Script/ControlRig.RigUnit_TransformConstraint(
    BaseBone=None
    BaseTransform=(Rotation=(X=0,Y=0,Z=0,W=1),Translation=(X=0,Y=0,Z=0),Scale3D=(X=1,Y=1,Z=1))
    BaseTransformSpace=GlobalSpace
    Bone=dyn_blenderHead_Base
    Targets="((Transform=(Rotation=(X=0,Y=0,Z=0,W=1),Translation=(X=0,Y=0,Z=0),Scale3D=(X=1,Y=1,Z=1)),Weight=0.8,bMaintainOffset=True,Filter=(TranslationFilter=(bX=True,bY=True,bZ=True),RotationFilter=(bX=True,bY=True,bZ=True),ScaleFilter=(bX=True,bY=True,bZ=True))),(T ...(truncated)"
    bUseInitialTransforms=True
    ExecutePin=@n1.ExecutePin
    Targets.0.Transform=&RigUnit_GetBoneTransform(Bone=spine_05).Transform
    Targets.1.Transform=&RigUnit_GetBoneTransform(Bone=neck_01).Transform
   ) {is=struct RigVMUnitNode}
   n2 = &RigVMFunction_MathQuaternionToEuler(RotationOrder=ZYX, Value=&RigUnit_GetRelativeBoneTransform(Bone=head, Space=spine_05).Transform.Rotation)
   > /Script/ControlRig.RigUnit_ModifyBoneTransforms(
    BoneToModify=((Bone="dyn_blenderHead",Transform=(Rotation=(X=0,Y=0,Z=0,W=1),Translation=(X=0,Y=0,Z=0),Scale3D=(X=1,Y=1,Z=1))))
    Mode=AdditiveLocal
    Weight=1
    WeightMaximum=1
    WeightMinimum=0
    BoneToModify.0.Transform.Rotation=&RigVMFunction_MathQuaternionFromEuler(Euler=/Script/RigVM.RigVMFunction_MathVectorMul(A=(X=1,Y=0,Z=0), B=@n2.Result) {is=struct RigVMUnitNode, template="Multiply::Execute(in A,in B,out Result)"}.Result).Result
   ) {is=struct RigVMUnitNode}
   > &RigUnit_ModifyBoneTransforms(
    BoneToModify=((Bone="dyn_blenderHead_extraSpin",Transform=(Rotation=(X=0,Y=0,Z=0,W=1),Translation=(X=0,Y=0,Z=0),Scale3D=(X=1,Y=1,Z=1))))
    Mode=AdditiveLocal
    Weight=1
    WeightMaximum=1
    WeightMinimum=0
    BoneToModify.0.Transform.Rotation=/Script/RigVM.RigVMFunction_MathQuaternionFromRotator(Rotator=RigVMVariableNode() {variable=blender}.Value) {is=struct RigVMUnitNode}.Result
   )
  ]}
]
hierarchy [
 root {loc=<0 0 0>, rot=<0 0 0 1>}
 attach {parent=root, loc=<0 0 0>, rot=<0 0 0 1>}
 pelvis {parent=root, loc=<0 -0.7071 93.1191>, rot=<0.0127 0.707 -0.0127 -0.707>}
 spine_01 {parent=pelvis, loc=<6.7494 -0.6388 0>, rot=<0 0 0.034 -0.9994>}
 spine_02 {parent=spine_01, loc=<7.2518 -0.035 0>, rot=<0 0 0.0238 -0.9997>}
 spine_03 {parent=spine_02, loc=<6.0104 0 0>, rot=<0 0 0.0149 -0.9999>}
 spine_04 {parent=spine_03, loc=<8.1024 0 0>, rot=<0 0 -0.0132 -0.9999>}
 spine_05 {parent=spine_04, loc=<10.0146 0 0>, rot=<0 0 -0.0015 -1>}
 clavicle_l {parent=spine_05, loc=<4.3575 -0.1466 -2.3223>, rot=<0.0465 -0.7289 -0.009 -0.6829>}
 clavicle_r {parent=spine_05, loc=<4.3575 -0.1466 2.3223>, rot=<0.7289 0.0465 -0.6829 0.009>}
 neck_01 {parent=spine_05, loc=<11.3138 0.1529 0>, rot=<0 0 -0.2003 -0.9797>}
 neck_02 {parent=neck_01, loc=<4.5627 0 0>, rot=<0 0 0.0504 -0.9987>}
 head {parent=neck_02, loc=<5.3914 0 0>, rot=<0 0 0.115 -0.9934>}
 faceAttach {parent=head, loc=<-2.7937 0.7863 0.0013>, rot=<0.0082 -0.7071 -0.0082 -0.7071>, scale=<0.89 0.89 0.89>}
 C_brow_mid {parent=faceAttach, loc=<-0.0221 9.7038 8.9876>, rot=<0 0.0818 -0.0162 -0.9965>}
 L_brow_mid {parent=faceAttach, loc=<3.0695 9.2654 9.7921>, rot=<-0.0003 -0.0056 0.2553 -0.9668>}
 L_brow_outer {parent=faceAttach, loc=<5.3419 7.2537 9.8446>, rot=<0.0641 0.0768 0.4254 -0.8995>}
 L_cheek_inner {parent=faceAttach, loc=<4.3023 7.7762 4.0542>, rot=<-0.0572 -0.2357 0.3229 -0.9148>}
 C_lip_upper_mid {parent=faceAttach, loc=<0.0005 9.8598 1.6991>, rot=<0.2225 0.0017 -0.0006 -0.9749>}
 L_lip_upper_outer {parent=faceAttach, loc=<1.8192 8.9546 1.6462>, rot=<0.0396 -0.0625 0.2588 -0.9631>}
 L_lip_corner {parent=faceAttach, loc=<2.7401 8.2941 1.4946>, rot=<0.0019 0.0091 -0.0145 -0.9999>}
 C_nose_bridge {parent=faceAttach, loc=<-0.0006 11.0223 6.2387>, rot=<-0.2218 -0.0232 -0.0065 -0.9748>}
 R_brow_mid {parent=faceAttach, loc=<-3.0664 9.2678 9.7975>, rot=<-0.0001 0.0043 -0.2553 -0.9669>}
 R_brow_outer {parent=faceAttach, loc=<-5.3497 7.2389 9.8305>, rot=<0.0455 -0.0534 -0.431 -0.8996>}
 R_cheek_inner {parent=faceAttach, loc=<-4.2872 7.7844 4.0459>, rot=<-0.0586 0.2393 -0.3225 -0.914>}
 R_lip_corner {parent=faceAttach, loc=<-2.6994 8.3193 1.5444>, rot=<0.0027 -0.0082 0.0146 -0.9999>}
 C_jaw {parent=faceAttach, loc=<-0.0634 2.9539 1.4309>, rot=<0 0 0 1>}
 L_lip_lower_outer {parent=C_jaw, loc=<1.9022 5.8076 0.1831>, rot=<0.0256 0.1132 0.3565 -0.9271>}
 C_lip_lower_mid {parent=C_jaw, loc=<0.0525 6.6939 0.139>, rot=<0 -0.003 -0.0006 -1>}
 R_lip_lower_outer {parent=C_jaw, loc=<-1.7542 5.923 0.1295>, rot=<0.0161 -0.0614 -0.3183 -0.9459>}
 teeth_lower {parent=C_jaw, loc=<-0.0244 4.6472 -1.6419>, rot=<0 0 0 1>}
 tongue {parent=C_jaw, loc=<0.0596 1.619 -0.7361>, rot=<-0.1568 0 0 -0.9876>}
 R_eye {parent=faceAttach, loc=<-2.965 6.9368 8.5542>, rot=<0 0 0.9998 -0.0187>}
 R_eye_lid_upper_mid {parent=R_eye, loc=<-0.0226 -1.1685 -0.0947>, rot=<0.0523 -0.0009 0.0167 -0.9985>}
 R_eye_lid_lower_mid {parent=R_eye, loc=<0.1236 -0.8584 -0.6512>, rot=<0.0329 -0.0005 0.0167 -0.9993>}
 L_eye {parent=faceAttach, loc=<2.9831 6.9338 8.5649>, rot=<0 0 0.9998 0.0187>}
 L_eye_lid_upper_mid {parent=L_eye, loc=<0.0243 -1.1744 -0.1315>, rot=<0.0523 0 0 -0.9986>}
 L_eye_lid_lower_mid {parent=L_eye, loc=<-0.0951 -0.8434 -0.7052>, rot=<0.0329 0 0 -0.9995>}
 teeth_upper {parent=faceAttach, loc=<0.0186 7.7912 2.3601>, rot=<0 0 0 1>}
 R_lip_upper_outer {parent=faceAttach, loc=<-1.8437 8.9968 1.7035>, rot=<0.0424 0.0546 -0.2558 -0.9643>}
 dyn_blenderHead_Base {parent=spine_05, loc=<14.9564 7.0798 -0.0393>, rot=<0.9971 0.0759 -0.0025 0.0036>}
 dyn_blenderHead {parent=dyn_blenderHead_Base, loc=<0 0 0>, rot=<0 0 0 1>}
 dyn_blenderHead_extraSpin {parent=dyn_blenderHead, loc=<0 0 0>, rot=<0 0 0 1>}
]
variables [blender {type=FRotator, default=(Pitch=0,Yaw=0,Roll=0)}]
