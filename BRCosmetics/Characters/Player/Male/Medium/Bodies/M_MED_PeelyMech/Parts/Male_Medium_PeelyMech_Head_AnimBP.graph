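# Animation Blueprint for the head part of the male medium PeelyMech character, driving blender head spin and the blender emote face.
# AnimGraph copies pose and curves from the attached parent mesh, then feeds headSpin into a Control Rig.
# Each update reads headSpin from the body mesh's extra-spin bone delta against the reference pose.
# It also checks the Peely Blender emote montage and pushes the PeelyBlenderFace curve into face material parameters.
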AnimBlueprint /BRCosmetics/Characters/Player/Male/Medium/Bodies/M_MED_PeelyMech/Parts/Male_Medium_PeelyMech_Head_AnimBP
target_skeleton=&/Male_Medium_PeelyMech_Head_Skeleton
graphs [
 AnimGraph {nodes [
   n1 = Control Rig!(Source=Copy Pose From Mesh() {bCopyCurves=true, bUseAttachedParent=true}, Alpha, blender:Rotator=Get headSpin() -> (headSpin:Rotator)) -> (Pose) {
    alphaBoolBlend {blendOption=Linear}
    controlRigAssetReference {blueprintRigClass=&/Male_Medium_PeelyMech_Head_CtrlRig.Male_Medium_PeelyMech_Head_CtrlRig_C}
   }
   Output Pose(Result=@n1) {layerGroup=None}
  ]}
 BlenderFace {nodes [BlenderFace() {is=title K2Node_FunctionEntry}]}
 EventGraph {nodes [
   Event BlueprintUpdateAnimation()
   n1 = Get PlayerPawn() -> (PlayerPawn:Fort Player Pawn Athena Object Reference)
   > Is Valid(InputObject=@n1)
   > n2 = Sequence()
   > Do Once()
   > Set BodySkelMesh(BodySkelMesh:Skeletal Mesh Component Object Reference=GetSkeletalMeshForPartType(self=@n1)) -> (Output_Get:Skeletal Mesh Component Object Reference)
   Set headSpin(
    execute=@n2.then_1
    headSpin:Rotator=GetDeltaTransformFromRefPose!(
     self=Get BodySkelMesh() -> (BodySkelMesh:Skeletal Mesh Component Object Reference)
     BoneName=dyn_blenderHead_extraSpin
     BaseName=dyn_blenderHead
    ) -> (ReturnValue_Location:Vector, ReturnValue_Rotation:Rotator, ReturnValue_Scale:Vector).ReturnValue_Rotation
   ) -> (Output_Get:Rotator)
   n3 = GetAnimInstance(self=GetOwnerSkeletalMeshComponent())
   > Set PeelyBlenderFaceActive(PeelyBlenderFaceActive:bool=Montage_IsActive!(
     self:Anim Instance Object Reference=@n3
     Montage:Anim Montage Object Reference=/BRCosmetics/Animation/Game/MainPlayer/Emotes/Peely_Blender/Emote_Peely_Blender_M
    ) -> (ReturnValue:bool)) -> (Output_Get:bool)
   > SetScalarParameterValueOnMaterials!(
    execute
    self:Mesh Component Object Reference=GetSkeletalMeshForPartType(self=Get PlayerPawn(), PartType=Face)
    ParameterName:Name=PeelyBlenderFace
    ParameterValue:float=GetCurveValue(self=@n3, CurveName=PeelyBlenderFace)
   ) -> (then)
   Event BlueprintInitializeAnimation()
   n4 = TryGetPawnOwner()
   > Is Valid(InputObject=@n4)
   > n5 = Cast To FortPlayerPawnAthena(Object=@n4)
   > Set PlayerPawn(PlayerPawn:Fort Player Pawn Athena Object Reference=@n5) -> (Output_Get:Fort Player Pawn Athena Object Reference)
  ]}
]
variables [
 BlenderFace {default=()}
 BodySkelMesh {type=SkeletalMeshComponent}
 headSpin {type=Rotator}
 headSpinBase {type=Rotator}
 PeelyBlenderFaceActive {type=boolean}
 PlayerPawn {type=FortPlayerPawnAthena}
]
